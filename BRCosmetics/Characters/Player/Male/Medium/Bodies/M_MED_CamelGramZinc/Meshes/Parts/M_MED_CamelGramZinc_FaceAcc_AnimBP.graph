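# Animation Blueprint for the face accessory mesh of the male medium CamelGramZinc player cosmetic body.
# It copies the pose from the attached parent mesh, then runs a Transform (Modify) Bone on the "hair" bone in component space.
# That node additively scales the bone by 0.9 in bone space, blended in only when bRigidBodyEnabled is false.
# The flag is not declared among this Blueprint's variables, and the event graph holds only an empty update event.
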AnimBlueprint /BRCosmetics/Characters/Player/Male/Medium/Bodies/M_MED_CamelGramZinc/Meshes/Parts/M_MED_CamelGramZinc_FaceAcc_AnimBP
target_skeleton=&/M_MED_CamelGramZinc_FaceAcc_Skeleton
graphs [
 AnimGraph {nodes [
   n1 = "Transform (Modify) Bone - Bone: hair"!(
    ComponentPose=Local To Component(LocalPose=Copy Pose From Mesh() {bUseAttachedParent=true})
    bAlphaBoolEnabled:bool=NOT Boolean(A=Get bRigidBodyEnabled())
    Translation="0,0,0"
    Rotation="0,0,0"
    Scale="0.9,0.9,0.9"
   ) -> (Pose) {
    alphaInputType=Bool
    rotationMode=BMM_Additive
    rotationSpace=BCS_BoneSpace
    scaleMode=BMM_Additive
    scaleSpace=BCS_BoneSpace
    translationMode=BMM_Additive
    translationSpace=BCS_BoneSpace
    alphaBoolBlend {blendOption=Linear}
    boneToModify {boneName=hair}
    scale=<0.9 0.9 0.9>
   }
   Output Pose(Result=Component To Local(ComponentPose=@n1)) {layerGroup=None}
  ]}
 EventGraph {nodes [Event BlueprintUpdateAnimation()]}
]
variables []
